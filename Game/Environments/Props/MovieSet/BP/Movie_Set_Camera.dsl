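; Movie set camera prop Blueprint whose camera component can be rotated through a variable.
; The construction script builds a rotator from CameraRotation on the yaw axis and applies it as the relative rotation of the Movieset_Camera component.
; BeginPlay, Tick and actor overlap events are empty, so the Blueprint adds no runtime behavior.
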(event EventBeginPlay)

(event Collision|EventActorBeginOverlap
  (params (OtherActor :type Actor%20Object%20Reference)))

(event EventTick
  (params (DeltaSeconds :type Float%20%28single-precision%29)))

(fn ConstructionScript
  (Transformation|SetRelativeRotation (Variables|Default|GetMovieset_Camera) (Math|Rotator|MakeRotator 0.0 0.0 (Variables|Default|GetCameraRotation)) false false))
